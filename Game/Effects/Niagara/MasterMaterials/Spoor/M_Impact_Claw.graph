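# Unlit, two-sided translucent master material for Niagara sprite claw impact effects, using a claw texture with switchable UV direction.
# Emissive blends claw channels tinted by Color1, adds screen-aligned nebula noise that fades out as glow is enabled, then multiplies by Color2 and particle color.
# Opacity combines particle alpha and glow-selected claw channels with a cloud-texture HeightLerp dissolve driven by a dynamic parameter.
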Material /Game/Effects/Niagara/MasterMaterials/Spoor/M_Impact_Claw
graph {
 properties {
  bUsedWithNiagaraSprites=true
  blendMode=BLEND_Translucent
  pixelDepthOffsetMode=PDOM_Legacy
  shadingModel=MSM_Unlit
  twoSided=true
  usedShadingModels=Unlit
 }
 nodes [
  n1 = ParticleColor()
  n2 = TextureCoordinate()
  n3 = TextureSampleParameter2D(UVs=LinearInterpolate(
    A=MaterialFunctionCall(XY=@n2) {materialFunction=&Swizzle}.YX
    B=@n2
    Alpha=ScalarParameter() {parameterName=ChangeDirection}
   )) {parameterName=Default_Tex, texture=/Game/Weapons/FORT_Melee/Pickaxe_ScareyBeary/FX/T_Pickaxe_ScareyBeary_Claw}
  n4 = Multiply(A=@n1.A, B=@n3.R)
  n5 = ScalarParameter() {defaultValue=1, parameterName=Use_TexR}
  n6 = Multiply(A=LinearInterpolate(A=@n3.G, B=@n3.G, Alpha=ScalarParameter() {parameterName=Use glow}))
  n7 = TextureSampleParameter2D(UVs=Multiply(A=MaterialFunctionCall!(Center Point World Position) {materialFunction=/Engine/Functions/Engine_MaterialFunctions02/UVs/WorldSpaceAlignedScreenCoordinates}.Offset and Scaled UVS) {constB=30}) {parameterName=Noise_Tex, texture=/Game/Characters/Player/Male/Medium/Bodies/M_Med_Soldier_04/Skins/Galaxy/Textures/T_FN_Nebula_style2}
  n8 = Multiply(A=@n7.RGB, B=ScalarParameter() {defaultValue=0.2, parameterName=NoiseTexture_Mul}) {constB=0.2}
  n9 = ScalarParameter() {parameterName=Use glow}
  n10 = Add(
   A=Multiply(A=VectorParameter() {parameterName=Color1, defaultValue=<rgba 1 0 0.5268 1>}.RGB, B=LinearInterpolate(A=@n4, B=@n6, Alpha=@n5))
   B=Multiply(A=LinearInterpolate(A=@n8, Alpha=@n9) {constB=0})
  )
  MaterialFunctionCall(Emissive=Multiply(A=VectorParameter() {parameterName=Color2, defaultValue=<rgba 8 0 7.2791 1>}.RGB, B=Multiply(A=@n10, B=@n1.RGB))) {materialFunction=&MF_FX_EmissiveAdjustments}
  n11 = Add(A=@n3.R, B=@n3.R)
  n12 = ScalarParameter() {parameterName=Use glow}
  n13 = MaterialFunctionCall(XY=TextureCoordinate() {vTiling=0.3}) {materialFunction=&Swizzle}
  n14 = Multiply(A=MaterialFunctionCall(
    Transition Phase=DynamicParameter() {paramNames [Dissolve,Param2,Param3,Param4]}.Dissolve
    Height Texture=TextureSample(UVs=@n13.YX) {samplerType=SAMPLERTYPE_LinearGrayscale, texture=/Game/Athena/Prototype/Terrain/T_CloudCoverage}.R
   ) {materialFunction=&HeightLerp}.Alpha) {constB=3}
  Multiply(
   A=LinearInterpolate(A=@n4, B=@n1.A, Alpha=ScalarParameter() {defaultValue=1, parameterName=Use_TexR})
   B=Multiply(A=LinearInterpolate(A=@n11, B=@n3.B, Alpha=@n12), B=Saturate(None=@n14))
  )
 ]
}
